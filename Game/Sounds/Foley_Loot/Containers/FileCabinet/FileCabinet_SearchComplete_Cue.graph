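# Sound cue for the file cabinet loot container's search-complete foley, routed to the loot collect sound class.
# A random node picks one of three equally weighted file cabinet search waves, then a modulator slightly varies its pitch and volume.
# It plays at reduced volume with prop search attenuation and its own concurrency limit that stops the farthest instance, then prevents new ones.
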SoundCue /Game/Sounds/Foley_Loot/Containers/FileCabinet/FileCabinet_SearchComplete_Cue
root=Modulator_15
nodes [
 {id=Modulator_15, children [Random_1], properties {pitchMax=1.07, pitchMin=0.92, volumeMax=1.07, volumeMin=0.92}}
 {id=Random_1, children [WavePlayer_6, WavePlayer_8, WavePlayer_10], properties {weights [1,1,1]}}
 {id=WavePlayer_10, properties {soundWaveAssetPtr=&/FileCabinet_Search05}}
 {id=WavePlayer_6, properties {soundWaveAssetPtr=&/FileCabinet_Search01}}
 {id=WavePlayer_8, properties {soundWaveAssetPtr=&/FileCabinet_Search03}}
]
properties {
 attenuationSettings=/Game/Sounds/Attenuation/Fort_Prop_Search_Att
 bOverrideConcurrency=true
 maxDistance=1300
 soundClassObject=&Fort_Loot_Collect
 volumeMultiplier=0.35
 concurrencyOverrides {maxCount=4, resolutionRule=StopFarthestThenPreventNew}
}
