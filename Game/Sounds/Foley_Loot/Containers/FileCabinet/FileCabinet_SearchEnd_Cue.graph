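# Sound cue that plays when searching a file cabinet loot container ends.
# A mixer layers the "no more loot" UI sound with a random pick from three file cabinet search foley waves.
# An enveloper ramps volume and pitch up over the foley's opening and varies both slightly per play.
# The cue is quiet, routed to the loot collect sound class with prop search attenuation and a small concurrency cap that stops the farthest instance first.
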SoundCue /Game/Sounds/Foley_Loot/Containers/FileCabinet/FileCabinet_SearchEnd_Cue
root=Mixer_12
nodes [
 {
  id=Enveloper_21
  children [Random_1]
  properties {
   pitchMax=1.07
   pitchMin=0.92
   volumeMax=1.07
   volumeMin=0.92
   pitchCurve {editorCurveData {keys [{interpMode=RCIM_Cubic, tangentMode=RCTM_Break, time=0, value=0.95}, {time=0.1}]}}
   volumeCurve {editorCurveData {keys [{interpMode=RCIM_Cubic, tangentMode=RCTM_Break, time=0, value=0.5}, <at 0.3 1>]}}
  }
 }
 {id=Mixer_12, children [WavePlayer_43, Enveloper_21], properties {inputVolume [0.8,0.8]}}
 {id=Random_1, children [WavePlayer_6, WavePlayer_8, WavePlayer_10], properties {weights [1,1,1]}}
 {id=WavePlayer_10, properties {soundWaveAssetPtr=&/FileCabinet_Search05}}
 {id=WavePlayer_43, properties {soundWaveAssetPtr=/Game/Sounds/Fort_UI_Sounds/Gameplay/Fort_UI_NoMoreLoot}}
 {id=WavePlayer_6, properties {soundWaveAssetPtr=&/FileCabinet_Search01}}
 {id=WavePlayer_8, properties {soundWaveAssetPtr=&/FileCabinet_Search03}}
]
properties {
 attenuationSettings=/Game/Sounds/Attenuation/Fort_Prop_Search_Att
 bOverrideConcurrency=true
 maxDistance=1300
 soundClassObject=&Fort_Loot_Collect
 volumeMultiplier=0.35
 concurrencyOverrides {maxCount=4, resolutionRule=StopFarthestThenPreventNew}
}
