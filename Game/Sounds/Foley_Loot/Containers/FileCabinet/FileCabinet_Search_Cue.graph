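# Sound cue for searching a file cabinet loot container, using the loot-collect sound class and prop-search attenuation.
# A mixer layers a random cabinet search wave with pitch and volume modulation, a quieter one-second delayed branch through that same modulator, and a louder random gear-rummage wave.
# Overridden concurrency prevents new instances from starting once the limit is reached.
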SoundCue /Game/Sounds/Foley_Loot/Containers/FileCabinet/FileCabinet_Search_Cue
root=Mixer_24
nodes [
 {id=Delay_13, children [Modulator_18], properties {delayMax=1, delayMin=1}}
 {id=Mixer_24, children [Modulator_18,Delay_13,Random_23], properties {inputVolume [0.5,0.2,1.2]}}
 {id=Modulator_18, children [Random_0], properties {pitchMax=1.3, pitchMin=0.92, volumeMax=1.07, volumeMin=0.92}}
 {id=Random_0, children [WavePlayer_0,WavePlayer_2,WavePlayer_4], properties {weights [1,1,1]}}
 {id=Random_23, children [WavePlayer_119, WavePlayer_120, WavePlayer_122], properties {weights [1,1,1]}}
 {id=WavePlayer_0, properties {soundWaveAssetPtr=&/FileCabinet_Search01}}
 {id=WavePlayer_119, properties {soundWaveAssetPtr=/Game/Sounds/Foley_Loot/Containers/Misc/Prop_Gear_Search_05}}
 {id=WavePlayer_120, properties {soundWaveAssetPtr=/Game/Sounds/Foley_Loot/Containers/Misc/Prop_Gear_Search_06}}
 {id=WavePlayer_122, properties {soundWaveAssetPtr=/Game/Sounds/Foley_Loot/Containers/Misc/Prop_Gear_Search_01}}
 {id=WavePlayer_2, properties {soundWaveAssetPtr=&/FileCabinet_Search03}}
 {id=WavePlayer_4, properties {soundWaveAssetPtr=&/FileCabinet_Search05}}
]
properties {
 attenuationSettings=/Game/Sounds/Attenuation/Fort_Prop_Search_Att
 bOverrideConcurrency=true
 maxDistance=1300
 soundClassObject=&Fort_Loot_Collect
 volumeMultiplier=0.7
 concurrencyOverrides {resolutionRule=PreventNew}
}
